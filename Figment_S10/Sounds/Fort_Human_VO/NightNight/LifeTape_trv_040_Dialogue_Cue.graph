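# Sound cue that plays the LifeTape_trv_040 dialogue line for the NightNight human voice-over set in Figment Season 10.
# A single dialogue player node plays the dialogue wave with LifeTape_DLV as speaker and no listed targets.
# It routes through the VO announcement sound class with a very large max distance.
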SoundCue /Figment_S10/Sounds/Fort_Human_VO/NightNight/LifeTape_trv_040_Dialogue_Cue
root=DialoguePlayer_0
nodes [{
  id=DialoguePlayer_0
  properties {dialogueWaveParameter {
    dialogueWave=&/LifeTape_trv_040_Dialogue
    context {speaker=&/LifeTape_DLV, targets []}
   }}
 }]
properties {maxDistance=2097152, soundClassObject=/Figment_S10/Sounds/FortSoundClassesAndModes/Fort_VO_Announcement_PSM, volumeMultiplier=1}
