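# Sound cue that plays the full LifeTape "trv" voice-over dialogue, part of the NightNight human VO set.
# A Concatenator plays six DialoguePlayer nodes back to back in numbered order, 010 through 060, all at equal input volume.
# Each player resolves its Dialogue Wave through the LifeTape_DLV speaker context with no targets.
# It routes to the announcement VO sound class and has a very large max distance.
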SoundCue /Figment_S10/Sounds/Fort_Human_VO/NightNight/LifeTape_trv_Full_Dialogue_Cue
root=Concatenator_0
nodes [
 {
  id=Concatenator_0
  children [DialoguePlayer_4, DialoguePlayer_5, DialoguePlayer_6, DialoguePlayer_7, DialoguePlayer_8, DialoguePlayer_9]
  properties {inputVolume [1,1,1,1,1,1]}
 }
 {
  id=DialoguePlayer_4
  properties {dialogueWaveParameter {
    dialogueWave=&/LifeTape_trv_010_Dialogue
    context {speaker=&/LifeTape_DLV, targets []}
   }}
 }
 {
  id=DialoguePlayer_5
  properties {dialogueWaveParameter {
    dialogueWave=&/LifeTape_trv_020_Dialogue
    context {speaker=&LifeTape_DLV, targets []}
   }}
 }
 {
  id=DialoguePlayer_6
  properties {dialogueWaveParameter {
    dialogueWave=&/LifeTape_trv_030_Dialogue
    context {speaker=&LifeTape_DLV, targets []}
   }}
 }
 {
  id=DialoguePlayer_7
  properties {dialogueWaveParameter {
    dialogueWave=&/LifeTape_trv_040_Dialogue
    context {speaker=&LifeTape_DLV, targets []}
   }}
 }
 {
  id=DialoguePlayer_8
  properties {dialogueWaveParameter {
    dialogueWave=&/LifeTape_trv_050_Dialogue
    context {speaker=&LifeTape_DLV, targets []}
   }}
 }
 {
  id=DialoguePlayer_9
  properties {dialogueWaveParameter {
    dialogueWave=&/LifeTape_trv_060_Dialogue
    context {speaker=&LifeTape_DLV, targets []}
   }}
 }
]
properties {maxDistance=2097152, soundClassObject=/Figment_S10/Sounds/FortSoundClassesAndModes/Fort_VO_Announcement_PSM, volumeMultiplier=1}
